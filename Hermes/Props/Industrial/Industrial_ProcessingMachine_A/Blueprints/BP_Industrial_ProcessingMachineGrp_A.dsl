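; Building prop Blueprint for the Industrial Processing Machine A group that tints one of its secondary mesh modules from building texture data.
; The construction script calls the parent, then fetches the building texture data array.
; A non-empty array sends the first entry's tint color to Module C; an empty one runs the same lookup for Module B.
; BeginPlay, overlap and Tick events are empty.
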(event EventBeginPlay)

(event Collision|EventActorBeginOverlap
  (params (OtherActor :type Actor%20Object%20Reference)))

(event EventTick
  (params (DeltaSeconds :type Float%20%28single-precision%29)))

(fn ConstructionScript
  (|CallParentFunction)
  (bind _outtexturedataarray (Building|GetBuildingTextureData))
  (if (> (Utilities|Array|Length _outtexturedataarray) 0)
    (Default|SetColoronSecondaryMesh (Math|Conversions|ToLinearColor(Color) (Variables|/Script/FortniteGame.BuildingTextureData|GetTintColor (Utilities|Array|Get(acopy) _outtexturedataarray))) (Variables|Default|GetSM_Industrial_ProcessingModule_C))
    (Default|SetColoronSecondaryMesh (Math|Conversions|ToLinearColor(Color) (Variables|/Script/FortniteGame.BuildingTextureData|GetTintColor (Utilities|Array|Get(acopy) _outtexturedataarray))) (Variables|Default|GetSM_Industrial_ProcessingModule_B))))
